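# Material function in the Concept_Inverted wheel cosmetic folder that hue-shifts an input Color by a scalar Angle.
# It uses Rodrigues' rotation to spin the RGB vector around the normalized (1,1,1) grey axis.
# The result sums the cosine-scaled color, the sine-scaled axis-color cross product, and the axis projection scaled by one minus cosine.
# Sine and Cosine keep their default period, so an Angle of 1 is a full turn.
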MaterialFunction /VehicleCosmetics/Wheels/Concept_Inverted/Materials/MF_HUE_Shift
graph {nodes [
  n1 = FunctionInput() {inputName=Angle, inputType=FunctionInput_Scalar}
  n2 = Cosine(None=@n1)
  n3 = Normalize(VectorInput=Constant3Vector() {constant=<rgba 1 1 1 0>})
  n4 = FunctionInput() {inputName=Color}
  n5 = Multiply(A=OneMinus(None=@n2), B=Multiply(A=@n3, B=DotProduct(A=@n3, B=@n4)))
  n6 = Multiply(A=@n2, B=@n4)
  n7 = Sine(None=@n1)
  n8 = CrossProduct(A=@n3, B=@n4)
  FunctionOutput(None=Add(A=Add(A=@n5, B=@n6), B=Multiply(A=@n7, B=@n8)))
 ]}
