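# Unlit masked material for the Concept Inverted vehicle wheel, blending emissive rim and tire looks by vertex red.
# Rim glow combines a mask texture with a panning UV1 gradient.
# Tire glow uses hue-shifted colors, camera-vector tinting, panning noise warp and an object-relative height gradient, with a painted switch.
# An optional LOD alpha mask drives opacity, and the result passes through shared vehicle damage and de-rez functions.
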Material /VehicleCosmetics/Wheels/Concept_Inverted/Materials/M_Wheel_ConceptInverted
graph {
 properties {
  bAutomaticallySetUsageInEditor=false
  bUseMaterialAttributes=true
  bUsedWithMeshDeformer=true
  bUsedWithSkeletalMesh=true
  blendMode=BLEND_Masked
  previewMesh=/VehicleCosmetics/Wheels/Glitch/Meshes/SK_Wheel_Glitch
  shadingModel=MSM_Unlit
  usedShadingModels=Unlit
 }
 nodes [
  n1 = LinearInterpolate(A=NamedRerouteUsage(), B=NamedRerouteUsage(), Alpha=VertexColor().R)
  n2 = TextureSample() {
   samplerSource=SSM_Wrap_WorldGroupSettings
   samplerType=SAMPLERTYPE_LinearColor
   texture=/VehicleCosmetics/Wheels/Concept_Inverted/Textures/T_ConceptInverted_Rim_M_LOD
  }
  n3 = Constant() {r=1}
  n4 = StaticSwitchParameter(True=LinearInterpolate(A=@n2.R, B=@n3, Alpha=VertexColor().R), False=@n3) {group=01_RIM, parameterName=LOds_Alpha}
  MaterialFunctionCall(In=MaterialFunctionCall(Material=SetMaterialAttributes!(MaterialAttributes, Emissive Color=@n1, Opacity Mask=@n4) {attributeSetTypes [#1, #2]}) {materialFunction=&MF_BaseVehicle_Damage}.Result) {materialFunction=&MF_VehicleDeRez_Transition}
  n5 = ScalarParameter() {defaultValue=0.5, group=01_RIM, parameterName=RIM_Emissive Multiplier, sortPriority=5}
  n6 = NamedRerouteUsage()
  n7 = Multiply(A=ScalarParameter() {defaultValue=0.5, group=01_RIM, parameterName=RIM_Bottom Emissive Intensity, sortPriority=2}, B=@n6)
  n8 = ScalarParameter() {defaultValue=1, group=01_RIM, parameterName=RIM_Top Emissive Intensity, sortPriority=1}
  n9 = VectorParameter() {group=01_RIM, parameterName=RIM_Custom Color, sortPriority=0, defaultValue=<rgba 0 0.4735 0.3613 1>}
  n10 = Add(
   A=Multiply(A=NamedRerouteUsage(), B=ScalarParameter() {defaultValue=-0.15, group=01_RIM, parameterName=RIM_Hue Shift, sortPriority=6})
   B=@n9.RGB
  )
  n11 = Add(A=@n7, B=LinearInterpolate(A=Multiply(A=@n10, B=@n8), B=@n7, Alpha=@n6))
  n12 = TextureCoordinate() {coordinateIndex=1}
  n13 = TextureSample(UVs=Panner(Coordinate=@n12) {speedY=-0.1}) {samplerSource=SSM_Wrap_WorldGroupSettings, texture=/VehicleCosmetics/Wheels/_SharedTextures/FX/GradientBottom2_T}
  n14 = TextureObjectParameter() {group=00_BASE, parameterName=Masks, texture=/VehicleCosmetics/Wheels/Concept_Inverted/Textures/T_Concept_Inverted_Rim_RGB}
  n15 = TextureSample(UVs=@n12, Tex=@n14) {samplerSource=SSM_Wrap_WorldGroupSettings}
  n16 = TextureSample(Tex=@n14) {samplerSource=SSM_Wrap_WorldGroupSettings}
  n17 = Clamp(None=Add(A=@n16.G, B=Multiply(A=@n13.R, B=@n15.R)))
  n18 = LinearInterpolate(
   A=Multiply(A=Clamp(None=@n11) {maxDefault=2.5, minDefault=-50}, B=@n17)
   B=Multiply(A=@n9.RGB, B=@n17)
   Alpha=ScalarParameter() {group=01_RIM, parameterName=RIM_Solid Color, sortPriority=3}
  )
  n19 = Multiply(A=@n18, B=@n16.G)
  n20 = ScalarParameter() {defaultValue=1, group=01_RIM, parameterName=RIM_Line Intensity, sortPriority=4}
  NamedRerouteDeclaration(None=LinearInterpolate(A=Multiply(A=@n5, B=@n18), B=Multiply(A=@n19, B=@n20), Alpha=@n16.G)) {name=RIM_Emissive, nodeColor=<rgba 0.0547 0 1 1>}
  NamedRerouteDeclaration(None=ScalarParameter() {defaultValue=-0.15, group=02_TIRE, parameterName=TIRE_Hue Shift, sortPriority=6}) {name=HueShift, nodeColor=<rgba 1 1 0 1>}
  n21 = Constant() {r=1}
  n22 = Rotator(Coordinate=Panner(Coordinate=TextureCoordinate()) {speedY=1}, Time=@n21) {speed=1.57}
  n23 = TextureObject() {texture=/VehicleCosmetics/Wheels/_SharedTextures/FX/DistortNoise_02_RGB_T}
  n24 = Multiply(A=TextureSample(UVs=@n22, Tex=@n23) {samplerSource=SSM_Wrap_WorldGroupSettings, texture=&DistortNoise_02_RGB_T}.R)
  n25 = TextureCoordinate()
  n26 = TextureSample(UVs=Rotator(Coordinate=Panner(Coordinate=@n25) {speedY=0.5}, Time=@n21) {speed=1.57}, Tex=@n23) {samplerSource=SSM_Wrap_WorldGroupSettings, texture=&DistortNoise_02_RGB_T}
  n27 = Panner(Coordinate=TextureCoordinate() {uTiling=7}) {speedY=0.3}
  n28 = Multiply(
   A=Add(A=Multiply(A=@n26.B), B=@n24)
   B=Multiply(A=TextureSample(UVs=@n27) {samplerSource=SSM_Wrap_WorldGroupSettings, texture=/VehicleCosmetics/Wheels/_SharedTextures/FX/Fin_RGB}.G)
  )
  n29 = Rotator(Coordinate=Panner(Coordinate=TextureCoordinate() {uTiling=2, vTiling=2}) {speedY=0.3}, Time=Constant() {r=1}) {speed=1.57}
  n30 = TextureObject() {texture=/VehicleCosmetics/Wheels/_SharedTextures/FX/RaysGradient_RGB_T}
  n31 = Add(
   A=Multiply(A=TextureSample(UVs=@n29, Tex=@n30) {samplerSource=SSM_Wrap_WorldGroupSettings, texture=&DistortNoise_02_RGB_T}.R)
   B=TextureSample(UVs=@n22, Tex=@n30) {samplerSource=SSM_Wrap_WorldGroupSettings, texture=&DistortNoise_02_RGB_T}.B
  )
  NamedRerouteDeclaration(None=Add(A=@n24, B=Add(A=@n28, B=@n31))) {name=Warp, nodeColor=<rgba 1 0 1 1>}
  n32 = VectorParameter() {group=02_TIRE, parameterName=TIRE_Custom Color, sortPriority=0, defaultValue=<rgba 0 0.4735 0.3613 1>}
  n33 = NamedRerouteUsage()
  n34 = Multiply(A=MaterialFunctionCall!(Color=@n32.RGB, Angle=OneMinus(None=@n33)) {materialFunction=&/MF_HUE_Shift}.Result)
  n35 = CameraVectorWS()
  n36 = DotProduct(A=@n35, B=Constant() {r=0.1})
  n37 = DotProduct(A=OneMinus(None=@n35), B=Constant() {r=0.2})
  n38 = Multiply(A=Add(A=Multiply(A=@n34, B=@n36), B=Multiply(A=@n32.RGB, B=@n37)))
  n39 = Multiply(A=MaterialFunctionCall(Color=@n32.RGB, Angle=NamedRerouteUsage()) {materialFunction=&MF_HUE_Shift}.Result)
  n40 = VectorParameter() {group=02_TIRE, parameterName=TIRE_Secondary Color, sortPriority=1, defaultValue=<rgba 0 0.4735 0.3613 1>}
  n41 = Multiply(A=Multiply(
    A=StaticSwitchParameter(True=@n39, False=@n40.RGB) {defaultValue=true, group=02_TIRE, parameterName=TIRE_Hue Shift, sortPriority=-2}
    B=NamedRerouteUsage()
   ))
  n42 = Multiply(A=NamedRerouteUsage()) {constB=-0.15}
  n43 = Multiply(
   A=ScalarParameter() {defaultValue=1, group=02_TIRE, parameterName=TIRE_Top Emissive Intensity, sortPriority=2}
   B=StaticSwitchParameter(True=Add(A=@n42, B=@n32.RGB), False=Add(A=@n38, B=@n41)) {group=02_TIRE, parameterName=TIRE_Painted, sortPriority=-1}
  )
  n44 = ComponentMask(None=Subtract(A=WorldPosition().XYZ, B=ObjectPositionWS())) {b=true}
  n45 = StaticSwitchParameter(
   True=NamedRerouteDeclaration(None=ConstantBiasScale(None=@n44) {bias=-51.2, scale=0.0219}) {name=Z Mask, nodeColor=<rgba 0 0 1 1>}
   False=ConstantBiasScale(None=@n44) {bias=-64, scale=0.0172}
  ) {group=02_TIRE, parameterName=TIRE_Painted, sortPriority=-1}
  n46 = ScalarParameter() {defaultValue=1.2, group=02_TIRE, parameterName=TIRE_Bottom Emissive Intensity, sortPriority=3}
  n47 = Multiply(
   A=Add(A=LinearInterpolate(A=@n43, B=@n45, Alpha=@n45), B=Multiply(A=@n45, B=@n46))
   B=Multiply(A=NamedRerouteUsage(), B=ScalarParameter() {defaultValue=0.5, group=02_TIRE, parameterName=TIRE_Pattern Strength, sortPriority=4})
  )
  NamedRerouteDeclaration(None=Multiply(
    A=Clamp(None=@n47) {maxDefault=2.5, minDefault=-50}
    B=ScalarParameter() {defaultValue=1.5, group=02_TIRE, parameterName=TIRE_Emissive Multiplier, sortPriority=5}
   )) {name=TIRE_Emissive, nodeColor=<rgba 1 0 0.6797 1>}
  NamedRerouteDeclaration(None=@n16.B) {name=Tire Pattern, nodeColor=<rgba 0 0.6016 1 1>}
 ]
}
